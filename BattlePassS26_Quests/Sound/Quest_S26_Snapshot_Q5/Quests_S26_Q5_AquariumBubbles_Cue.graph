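# Sound cue for looping aquarium bubble ambience used by a Battle Pass Season 26 snapshot quest (Q5).
# A single looping wave player feeds the output directly with the aquarium bubbles wave.
# It routes through the quests sound class and borrows the Hidden Star ambient loop attenuation from supply drop sounds for spatial falloff.
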SoundCue /BattlePassS26_Quests/Sound/Quest_S26_Snapshot_Q5/Quests_S26_Q5_AquariumBubbles_Cue
root=WavePlayer_0
nodes [{
  id=WavePlayer_0
  properties {bLooping=true, soundWaveAssetPtr=&/Quests_S26_Q5_AquariumBubbles}
 }]
properties {
 attenuationSettings=/Game/Athena/Sounds/SupplyDrops/Fort_HiddenStar_Ambient_Loop_Att
 maxDistance=2500
 soundClassObject=&Fort_Quests
 volumeMultiplier=1
}
