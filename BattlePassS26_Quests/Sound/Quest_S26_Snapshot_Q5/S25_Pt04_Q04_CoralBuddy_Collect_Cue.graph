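# Sound cue for collecting a Coral Buddy in a Battle Pass quest, routed to the Fort_Quests sound class.
# A mixer layers a random fish-glider cheer, a quieter random cage-close sound and a reward-screen chest-upgrade sting at the lowest level.
# The sting starts after a short fixed delay, and a modulator at the root adds slight pitch and volume variation to the whole mix.
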SoundCue /BattlePassS26_Quests/Sound/Quest_S26_Snapshot_Q5/S25_Pt04_Q04_CoralBuddy_Collect_Cue
root=Modulator_0
nodes [
 {id=Delay_0, children [WavePlayer_17], properties {delayMax=0.1, delayMin=0.1}}
 {id=Mixer_1, children [Random_1,Random_3,Delay_0], properties {inputVolume [1,0.64,0.28]}}
 {id=Modulator_0, children [Mixer_1], properties {pitchMax=1.02, pitchMin=0.98, volumeMax=1.01, volumeMin=1}}
 {
  id=Random_1
  children [
   WavePlayer_1
   WavePlayer_2
   WavePlayer_18
   WavePlayer_19
   WavePlayer_20
   WavePlayer_21
   WavePlayer_22
   WavePlayer_23
   WavePlayer_24
   WavePlayer_25
   WavePlayer_26
   WavePlayer_27
  ]
  properties {weights [1,1,1,1,1,1,1,1,1,1,1,1]}
 }
 {id=Random_3, children [WavePlayer_28, WavePlayer_29, WavePlayer_30], properties {weights [1,1,1]}}
 {id=WavePlayer_1, properties {soundWaveAssetPtr=&/Glider_TeriyakiFish_Cheers_11}}
 {id=WavePlayer_17, properties {soundWaveAssetPtr=/Game/Sounds/Fort_UI_Sounds/Rewards/Fort_RewardScreen_TeamScore_ChestLootUpgrade}}
 {id=WavePlayer_18, properties {soundWaveAssetPtr=&/Glider_TeriyakiFish_Cheers_09}}
 {id=WavePlayer_19, properties {soundWaveAssetPtr=&/Glider_TeriyakiFish_Cheers_08}}
 {id=WavePlayer_2, properties {soundWaveAssetPtr=&/Glider_TeriyakiFish_Cheers_10}}
 {id=WavePlayer_20, properties {soundWaveAssetPtr=&/Glider_TeriyakiFish_Cheers_07}}
 {id=WavePlayer_21, properties {soundWaveAssetPtr=&/Glider_TeriyakiFish_Cheers_06}}
 {id=WavePlayer_22, properties {soundWaveAssetPtr=&/Glider_TeriyakiFish_Cheers_05}}
 {id=WavePlayer_23, properties {soundWaveAssetPtr=&/Glider_TeriyakiFish_Cheers_04}}
 {id=WavePlayer_24, properties {soundWaveAssetPtr=&/Glider_TeriyakiFish_Cheers_03}}
 {id=WavePlayer_25, properties {soundWaveAssetPtr=&/Glider_TeriyakiFish_Cheers_02}}
 {id=WavePlayer_26, properties {soundWaveAssetPtr=&/Glider_TeriyakiFish_Cheers_01}}
 {id=WavePlayer_27, properties {soundWaveAssetPtr=&/Glider_TeriyakiFish_Cheers_00}}
 {id=WavePlayer_28, properties {soundWaveAssetPtr=&/S25_Q04_Cage_Close_01}}
 {id=WavePlayer_29, properties {soundWaveAssetPtr=&/S25_Q04_Cage_Close_02}}
 {id=WavePlayer_30, properties {soundWaveAssetPtr=&/S25_Q04_Cage_Close_03}}
]
properties {maxDistance=2097152, soundClassObject=&Fort_Quests, volumeMultiplier=1}
