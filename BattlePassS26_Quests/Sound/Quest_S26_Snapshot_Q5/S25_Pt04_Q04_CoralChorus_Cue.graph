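# Looping Sound Cue for the Coral Buddies chorus singing, used as quest audio under the Season 26 Snapshot Q5 folder with Season 25 naming.
# A single looping Wave Player feeds the output directly and routes to the Fort_Quests sound class.
# Attenuation is overridden on the cue with a box shape, occlusion and wide stereo spread.
# The cue's debug flag is left on.
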SoundCue /BattlePassS26_Quests/Sound/Quest_S26_Snapshot_Q5/S25_Pt04_Q04_CoralChorus_Cue
root=WavePlayer_0
nodes [{
  id=WavePlayer_0
  properties {bLooping=true, soundWaveAssetPtr=&/S25_Q04_CoralBuddiesSinging_LP}
 }]
properties {
 bDebug=true
 bOverrideAttenuation=true
 maxDistance=650
 soundClassObject=&Fort_Quests
 volumeMultiplier=1
 attenuationOverrides {attenuationShape=Box, bEnableOcclusion=true, falloffDistance=650, stereoSpread=2300, attenuationShapeExtents {x=0, z=-200}}
}
